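; Gameplay ability for the Vivid Razor Explosive medallion that makes the first shot of a magazine explode on hit.
; It binds to the player's OnDidDamage and compares clip size, remaining ammo and cost per shot on the damage-causing ranged weapon.
; Targets tagged ExplosiveDamage are skipped; others get a burst cue and two target-selection passes applying effect containers.
; GetCustomAbilitySourceTransform returns the stored victim transform, centering targeting on the hit actor.
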(event Ability|EventActivateAbility
  (|CallParentFunction)
  (bind _outfortplayerpawn (CallFunction|GetAvatarasValidFortPlayerPawn)
    (:ValidFortPlayerPawn
      (bind _output_get (Variables|Default|SetPlayer _outfortplayerpawn))
      (EventDispatchers|BindEventtoOnDidDamage _output_get (EventDispatchers|CreateEvent :!event Custom|DidDamage)))))

(event Custom|DidDamage
  (params (DamagedActor :type Actor%20Object%20Reference) (Damage :type Float%20%28single-precision%29) (InstigatedBy :type Controller%20Object%20Reference) (DamageCauser :type Actor%20Object%20Reference) (HitLocation :type Vector) (FHitComponent :type Primitive%20Component%20Object%20Reference) (BoneName :type Name) (Momentum :type Vector))
  (Variables|Default|SetDamageCauser DamageCauser)
  (bind _output_get (Variables|Default|SetHitTarget DamagedActor))
  (bind _returnvalue (GameplayTags|HasMatchingGameplayTag(Message) _output_get "(TagName=\"Granted.VividRazorMedallions.Explosive.ExplosiveDamage\")"))
  (if (and (not _returnvalue) (CallFunction|IsFirstShot))
    (bind _hittarget (Variables|Default|GetHitTarget)
      (:then
        (bind (_returnvalue_location _returnvalue_rotation) (Math|Transform|BreakTransform (Transformation|GetActorTransform _hittarget)))
        (Variables|SetTargetTransform_Location (Math|Transform|MakeTransform _returnvalue_location _returnvalue_rotation))
        (GameplayCue|ExecuteGameplayCueOnActor(Burst) _hittarget (Variables|Default|GetGameplay%20Cue%20Tag) (Ability|GameplayCue|MakeGameplayCueParameters :NormalizedMagnitude 0.0 :RawMagnitude 0.0 :Location HitLocation :Normal "0,0,0" :GameplayEffectLevel 1 :AbilityLevel 1 :bReplicateLocationWhenUsingMinimalRepProxy false))
        (bind (_asynctaskproxy _targetdata _applicationtag) (Ability|Tasks|DoTargetSelection "None" "(TagName=\"Targeting.ApplicationTag.0\")" true)
          (:then
            (bind (_asynctaskproxy_2 _targetdata_2 _applicationtag_2) (Ability|Tasks|DoTargetSelection "None" "(TagName=\"Gameplay.Application.Explode.Additonal\")" true)
              (:Targeted
                (Ability|ApplyGameplayEffectContainer _targetdata_2 _applicationtag_2))))
          (:Targeted
            (Ability|ApplyGameplayEffectContainer _targetdata _applicationtag))))
      (:else))))

(fn GetCustomAbilitySourceTransform
  (return (Variables|Default|GetTargetTransform)))

(fn IsFirstShot
  (bind _damagecauser (Variables|Default|GetDamageCauser)
    (:then
      (bind _asfort_weapon_ranged (Utilities|Casting|CastToFortWeaponRanged _damagecauser)
        (:then
          (return (not (Math|Conversions|ToBoolean(Integer) (- (- (FortWeapon|GetBulletsPerClip _asfort_weapon_ranged) (FortWeapon|GetMagazineAmmoCount _asfort_weapon_ranged)) (FortWeapon|GetAmmoCostPerFire _asfort_weapon_ranged 0.0))))))
        (:CastFailed
          _)))
    (:else
      _))
  (return false))
